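# Sound cue for the Cane event in Athena that plays a medium earthquake shake rumble in the scripted-events sound class.
# The wave player feeds an Enveloper whose volume curve holds full level, then fades to silence over a few seconds.
# A continuous modulator at the root exposes an Intensity parameter, so the volume can be scaled at runtime.
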SoundCue /FNGameplayCues/Athena/Sounds/Events/Cane/Cane_Cue
root=ModulatorContinuous_0
nodes [
 {
  id=Enveloper_0
  children [WavePlayer_10]
  properties {pitchCurve {editorCurveData {keys [{time=0}]}}, volumeCurve {editorCurveData {keys [<at 0 1>,<at 2 1>,<at 5 0>]}}}
 }
 {id=ModulatorContinuous_0, children [Enveloper_0], properties {volumeModulationParams {parameterName=Intensity}}}
 {id=WavePlayer_10, properties {soundWaveAssetPtr=/FNGameplayCues/Sounds/Fort_Impact_Sounds/Demolition/Earthquake/EarthQuake_Shakes_Med}}
]
properties {maxDistance=2097152, soundClassObject=/Game/Sounds/FortSoundClassesAndModes/Fort_Events_Scripted, volumeMultiplier=1}
