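# Sound cue for the interior sound of the Cane event in Athena, routed to the scripted events sound class.
# A single wave player plays Cane_Inside_01 through a continuous modulator.
# The modulator maps the Intensity parameter to volume, so whatever triggers the cue can scale its loudness at runtime.
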SoundCue /FNGameplayCues/Athena/Sounds/Events/Cane/Cane_Inside_Cue
root=ModulatorContinuous_1
nodes [
 {
  id=ModulatorContinuous_1
  children [WavePlayer_3]
  properties {volumeModulationParams {default=0, minInput=0.25, minOutput=0.3, parameterName=Intensity}}
 }
 {id=WavePlayer_3, properties {soundWaveAssetPtr=&/Cane_Inside_01}}
]
properties {maxDistance=2097152, soundClassObject=/Game/Sounds/FortSoundClassesAndModes/Fort_Events_Scripted}
